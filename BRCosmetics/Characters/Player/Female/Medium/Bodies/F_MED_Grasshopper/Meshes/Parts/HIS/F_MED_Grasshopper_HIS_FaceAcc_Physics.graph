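# Physics asset for the face accessory mesh of the female Grasshopper character cosmetic, covering dynamic antenna and hair bones.
# Capsule bodies form two mirrored four-link antenna chains and two parallel hair bodies, each group hanging from a master bone constrained to the head.
# Master-to-head joints leave swing and twist free, while downstream joints use tight angular limits that loosen toward the antenna tips.
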PhysicsAsset /BRCosmetics/Characters/Player/Female/Medium/Bodies/F_MED_Grasshopper/Meshes/Parts/HIS/F_MED_Grasshopper_HIS_FaceAcc_Physics
bodies {
 Dyn_Antenna_l1 [{shapeType=Capsule, length=3, radius=1, shapeName=Dyn_Antenna_l1_capsule, center=<1 0 0>, rotation=<rot 90 0 0>}]
 Dyn_Antenna_l2 [{shapeType=Capsule, length=3, radius=1, shapeName=Dyn_Antenna_l2_capsule, center=<1 0 0>, rotation=<rot 90 0 0>}]
 Dyn_Antenna_l3 [{shapeType=Capsule, length=3, radius=1, shapeName=Dyn_Antenna_l3_capsule, center=<1 0 0>, rotation=<rot 90 0 0>}]
 Dyn_Antenna_l4 [{shapeType=Capsule, length=3, radius=1, shapeName=Dyn_Antenna_l4_capsule, center=<1 0 0>, rotation=<rot 90 0 0>}]
 Dyn_Antenna_r1 [{shapeType=Capsule, length=3, radius=1, shapeName=Dyn_Antenna_r1_capsule, center=<1 0 0>, rotation=<rot 90 0 0>}]
 Dyn_Antenna_r2 [{shapeType=Capsule, length=3, radius=1, shapeName=Dyn_Antenna_r2_capsule, center=<1 0 0>, rotation=<rot 90 0 0>}]
 Dyn_Antenna_r3 [{shapeType=Capsule, length=3, radius=1, shapeName=Dyn_Antenna_r3_capsule, center=<1 0 0>, rotation=<rot 90 0 0>}]
 Dyn_Antenna_r4 [{shapeType=Capsule, length=3, radius=1, shapeName=Dyn_Antenna_r4_capsule, center=<1 0 0>, rotation=<rot 90 0 0>}]
 Dyn_Hair1 [{shapeType=Capsule, length=3, radius=1, shapeName=Dyn_Hair1_capsule, center=<0.5 0 0>, rotation=<rot 90 0 0>}]
 Dyn_Hair2 [{shapeType=Capsule, length=3, radius=1, shapeName=Dyn_Hair2_capsule, center=<0.5 0 0>, rotation=<rot 90 0 0>}]
 Dyn_Master_Antenna_l [{shapeType=Capsule, length=0.5, radius=0.5, shapeName=Dyn_Master_Antenna_l_capsule}]
 Dyn_Master_Antenna_r [{shapeType=Capsule, length=0.5, radius=0.5, shapeName=Dyn_Master_Antenna_r_capsule}]
 Dyn_Master_Hair [{shapeType=Capsule, length=0.5, radius=0.5, shapeName=Dyn_Master_Hair_capsule}]
 head [{shapeType=Capsule, length=0.5, radius=0.5, shapeName=head_capsule}]
}
constraints [
 {
  bone1Name=Dyn_Master_Antenna_l
  bone2Name=head
  swing1LimitDegrees=45
  swing1Motion=Free
  swing2LimitDegrees=45
  swing2Motion=Free
  twistLimitDegrees=45
  twistMotion=Free
 }
 {
  bone1Name=Dyn_Master_Antenna_r
  bone2Name=head
  swing1LimitDegrees=45
  swing1Motion=Free
  swing2LimitDegrees=45
  swing2Motion=Free
  twistLimitDegrees=45
  twistMotion=Free
 }
 {
  bone1Name=Dyn_Master_Hair
  bone2Name=head
  swing1LimitDegrees=45
  swing1Motion=Free
  swing2LimitDegrees=45
  swing2Motion=Free
  twistLimitDegrees=45
  twistMotion=Free
 }
 {bone1Name=Dyn_Antenna_l2, bone2Name=Dyn_Antenna_l1, swing1LimitDegrees=10, swing2LimitDegrees=5, twistLimitDegrees=1}
 {bone1Name=Dyn_Antenna_l3, bone2Name=Dyn_Antenna_l2, swing1LimitDegrees=20, swing2LimitDegrees=15, twistLimitDegrees=5}
 {bone1Name=Dyn_Antenna_l4, bone2Name=Dyn_Antenna_l3, swing1LimitDegrees=20, swing2LimitDegrees=15, twistLimitDegrees=5}
 {bone1Name=Dyn_Antenna_l1, bone2Name=Dyn_Master_Antenna_l, swing1LimitDegrees=10, swing2LimitDegrees=1, twistLimitDegrees=1}
 {bone1Name=Dyn_Antenna_r1, bone2Name=Dyn_Master_Antenna_r, swing1LimitDegrees=10, swing2LimitDegrees=1, twistLimitDegrees=1}
 {bone1Name=Dyn_Antenna_r2, bone2Name=Dyn_Antenna_r1, swing1LimitDegrees=10, swing2LimitDegrees=5, twistLimitDegrees=1}
 {bone1Name=Dyn_Antenna_r3, bone2Name=Dyn_Antenna_r2, swing1LimitDegrees=20, swing2LimitDegrees=15, twistLimitDegrees=5}
 {bone1Name=Dyn_Antenna_r4, bone2Name=Dyn_Antenna_r3, swing1LimitDegrees=20, swing2LimitDegrees=15, twistLimitDegrees=5}
 {bone1Name=Dyn_Hair1, bone2Name=Dyn_Master_Hair, swing1LimitDegrees=20, swing2LimitDegrees=5, twistLimitDegrees=1}
 {bone1Name=Dyn_Hair2, bone2Name=Dyn_Master_Hair, swing1LimitDegrees=20, swing2LimitDegrees=5, twistLimitDegrees=1}
]
