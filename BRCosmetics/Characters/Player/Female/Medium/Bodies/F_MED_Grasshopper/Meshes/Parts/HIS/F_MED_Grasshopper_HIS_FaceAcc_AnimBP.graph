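# Animation Blueprint for the F_MED_Grasshopper HIS face accessory, adding physics motion to a pose copied from the attached parent.
# Pectoral and clavicle branches blend in from the pawn's body mesh, found when customization loads.
# Three neck_01-relative RigidBody passes run in series, then an additive nudge offsets Dyn_Hood_r.
# No Pack weights the Back and BackOff passes, but both backbling-check branches set it and the unused Pack to 1.
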AnimBlueprint /BRCosmetics/Characters/Player/Female/Medium/Bodies/F_MED_Grasshopper/Meshes/Parts/HIS/F_MED_Grasshopper_HIS_FaceAcc_AnimBP
target_skeleton=&/F_MED_Grasshopper_HIS_FaceAcc_Skeleton
graphs [
 AnimGraph {nodes [
   n1 = RigidBody(ComponentPose=Local To Component(LocalPose=Use cached pose 'Start Cache'()), Alpha=Get No Pack() -> (No Pack:double)) {
    bOverrideWorldGravity=true
    overridePhysicsAsset=&/F_MED_Grasshopper_Back_HIS_FaceAcc_Physics
    simulationSpace=BaseBoneSpace
    alphaBoolBlend {blendOption=Linear}
    baseBoneRef {boneName=neck_01}
    componentAppliedLinearAccClamp=<3000 3000 3000>
    componentLinearAccScale=<0.3 0.3 0.3>
    componentLinearVelScale=<0.3 0.3 0.3>
    overrideWorldGravity {z=-980}
   }
   n2 = RigidBody(ComponentPose=@n1, Alpha=Get No Pack()) {
    bOverrideWorldGravity=true
    overridePhysicsAsset=&/F_MED_Grasshopper_BackOff_HIS_FaceAcc_Physics
    simulationSpace=BaseBoneSpace
    alphaBoolBlend {blendOption=Linear}
    baseBoneRef {boneName=neck_01}
    componentAppliedLinearAccClamp=<3000 3000 3000>
    componentLinearAccScale=<0.3 0.3 0.3>
    componentLinearVelScale=<0.3 0.3 0.3>
    overrideWorldGravity {z=-980}
   }
   n3 = RigidBody(ComponentPose=@n2) {
    bFreezeIncomingPoseOnStart=true
    bOverrideWorldGravity=true
    overridePhysicsAsset=&/F_MED_Grasshopper_HIS_FaceAcc_Physics
    simulationSpace=BaseBoneSpace
    alphaBoolBlend {blendOption=Linear}
    baseBoneRef {boneName=neck_01}
    componentAppliedLinearAccClamp=<2000 2000 2000>
    componentLinearAccScale=<0.4 0.3 0.3>
    componentLinearVelScale=<0.4 0.3 0.3>
    overrideWorldGravity {z=-980}
   }
   n4 = "Transform (Modify) Bone - Bone: Dyn_Hood_r"(ComponentPose=@n3, Alpha=1, Translation="0,0.7,0", Rotation="0,0,0", Scale="1,1,1") {
    rotationMode=BMM_Additive
    rotationSpace=BCS_BoneSpace
    translationMode=BMM_Additive
    translationSpace=BCS_BoneSpace
    alphaBoolBlend {blendOption=Linear}
    boneToModify {boneName=Dyn_Hood_r}
    translation {y=0.7}
   }
   Output Pose(Result=Component To Local(ComponentPose=@n4)) {layerGroup=None}
   n5 = Layered blend per bone(
    BasePose=Copy Pose From Mesh() {bUseAttachedParent=true}
    BlendPoses_0=Copy Pose From Mesh(SourceMeshComponent=Get Body Mesh() -> (Body Mesh:Skeletal Mesh Component Object Reference)) {bUseAttachedParent=true}
   ) {layerSetup [{branchFilters [{boneName=deform_pectoral_fr_l}, {boneName=deform_pectoral_fr_r}, {boneName=clavicle_l}, {boneName=clavicle_r}]}]}
   Save cached pose 'Start Cache'(Pose=@n5) {cacheName=Start Cache}
  ]}
 EventGraph {nodes [
   Event OnCharacterCustomizationLoaded()
   > n1 = Cast To FortPlayerPawnAthena(Object=TryGetPawnOwner())
   > Set Body Mesh(Body Mesh:Skeletal Mesh Component Object Reference=GetSkeletalMeshForPartType(self=@n1)) -> (Output_Get:Skeletal Mesh Component Object Reference)
   > n2 = Branch(Condition=CheckForBackbling())
   > Set No Pack(No Pack:double=1) -> (Output_Get:double)
   > Set Pack(Pack:double=1) -> (Output_Get:double)
   Set Pack(execute=@n2.else, Pack=1)
   > Set No Pack(No Pack=1)
   Event BlueprintUpdateAnimation()
  ]}
]
variables [
 Body Mesh {type=SkeletalMeshComponent}
 No Pack {type=number}
 Pack {type=number}
]
